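# FortReplayCameraOption is an IconTextButton child widget that acts as one selectable camera option in the replay UI.
# On construct it sets its button text from the replay context's display name for its CameraType variable.
# On click it casts the local player controller to FortPlayerControllerSpectating and switches to that camera type without requesting a reset.
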WidgetBlueprint /Game/UI/Replay/FortReplayCameraOption
graphs [EventGraph {nodes [
   Event On Clicked()
   > n1 = Cast To FortPlayerControllerSpectating!(
    execute
    Object=Get Local Player Controller From ID!(ControllerID:int=0) -> (ReturnValue:Player Controller Object Reference) {is=title K2Node_CallFunction}
   ) -> (then, CastFailed, AsFort Player Controller Spectating:Fort Player Controller Spectating Object Reference) {is=title K2Node_DynamicCast}
   > SetNewCameraType!(
    execute
    self:Fort Player Controller Spectating Object Reference=@n1
    NewCameraType:ESpectatorCameraType Enum=Get CameraType() -> (CameraType:ESpectatorCameraType Enum=ThirdPerson)
    bReset:bool=false
   ) -> (then)
   Event Construct()
   > Parent: Construct()
   > Set Text!(
    execute
    self:Icon Text Button Object Reference
    ButtonText:Text=GetCameraNameFromType!(
     self:Fort Replay Context Object Reference=Get FortReplayContext!() -> (ReturnValue:Fort Replay Context Object Reference) {is=title K2Node_GetBlueprintContextEx}
     CamType:ESpectatorCameraType Enum=Get CameraType() -> (CameraType=ThirdPerson)
    ) -> (ReturnValue:Text)
   ) -> (then) {is=title K2Node_CallFunction}
  ]}]
variables [CameraType {type=ESpectatorCameraType, default=Gameplay}]
widgets {
 info {inheritedWidgetCount=0, parentClass=/FortUILibrary/Foundation/Buttons/IconTextButton.IconTextButton_C, rootWidgetClass=null, widgetCount=0}
 widgets []
}
