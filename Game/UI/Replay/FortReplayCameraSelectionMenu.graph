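# Replay camera selection menu widget that lists the spectator camera modes available to the owning player.
# On Construct it casts the owning player to FortPlayerControllerSpectating and loops over that controller's CameraModeCycle array.
# Each mode spawns a FortReplayCameraOption widget, passed its camera type and appended to the CameraList vertical box inside a translucent dark border.
# The menu only builds the list, with no selection handling of its own.
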WidgetBlueprint /Game/UI/Replay/FortReplayCameraSelectionMenu
graphs [EventGraph {nodes [
   Event Construct()
   > n1 = Cast To FortPlayerControllerSpectating!(execute, Object=GetOwningPlayer()) -> (then, CastFailed, AsFort Player Controller Spectating:Fort Player Controller Spectating Object Reference) {is=title K2Node_DynamicCast}
   > n2 = For Each Loop(Array:Array of ESpectatorCameraType Enums=Get CameraModeCycle!(self:Fort Player Controller Spectating Object Reference=@n1) -> (CameraModeCycle:Array of ESpectatorCameraType Enums=ThirdPerson)) -> (Array Element:ESpectatorCameraType Enum)
   > n3 = Create Widget!(
    execute:Exec
    Class:User Widget Class Reference=&/FortReplayCameraOption.FortReplayCameraOption_C
    OwningPlayer:Player Controller Object Reference
    CameraType:ESpectatorCameraType Enum=@n2
    ButtonText:Text="INVTEXT(\"{Text}\")"
    IconBrush:Slate Brush Structure
    ControllerInputStyle:Common Button Style Class Reference=/FortUILibrary/Foundation/Buttons/ButtonStyles/ButtonStyle-TransparentWithCues-XS-B_OverlayYellow.ButtonStyle-TransparentWithCues-XS-B_OverlayYellow_C
    ButtonTextJustification:ETextJustify Enum=Center
    bDisplayAllCaps:bool=False
    Font:Slate Font Info Structure
    Style:Common Button Style Class Reference=&ButtonStyle-TransparentWithCues-XS-B_OverlayYellow_C
    bApplyAlphaOnDisable:bool=False
    bLocked:bool=False
    bSelectable:bool=False
    bShouldSelectUponReceivingFocus:bool=False
    bInteractableWhenSelected:bool=False
    bToggleable:bool=False
    bTriggerClickedAfterSelection:bool=False
    bDisplayInputActionWhenNotInteractable:bool=True
    bHideInputActionWithKeyboard:bool=False
    bShouldUseFallbackDefaultInputAction:bool=True
    bRequiresHold:bool=False
    HoldData:Common UIHold Data Class Reference
    ClickMethod:EButtonClickMethod Enum=DownAndUp
    InputPriority:int=0
    TriggeringInputAction:Data Table Row Handle Structure
   ) -> (then:Exec, ReturnValue:Fort Replay Camera Option Object Reference)
   > AddChild(self=Get CameraList() -> (CameraList:Vertical Box Object Reference), Content=@n3)
  ]}]
variables []
widgets {
 info {inheritedWidgetCount=0, parentClass=&UserWidget, rootWidgetClass=&Border, widgetCount=2}
 widgets [
  Border_0 {widgetClassPath=&Border, widget=Border_0, properties {brushColor=<rgba 0.0176 0.0232 0.0331 0.5>, padding=<ltrb 0 0 0 0>}}
  CameraList {
   widgetClassPath=&VerticalBox
   bIsVariable=true
   parent=Border_0
   widget=CameraList
   slot {class=BorderSlot, properties {padding=<ltrb 0 0 0 0>}}
  }
 ]
}
