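# Replay UI button widget, derived from IconTextButton, that represents one player the spectator can choose to follow.
# SetupName stores the given name in PlayerName and shows it as the button text.
# On click, it casts the owning player to FortPlayerControllerSpectating and calls FollowPlayerByName with the stored name.
# It adds no widgets of its own, and Construct only calls the parent.
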WidgetBlueprint /Game/UI/Replay/FortReplayPlayerOption
graphs [EventGraph {nodes [
   n1 = SetupName!() -> (OutputDelegate, then, ButtonText:Text) {is=title K2Node_CustomEvent}
   > n2 = Set PlayerName(PlayerName:Text=@n1.ButtonText) -> (Output_Get:Text)
   > Set Text!(execute, self:Icon Text Button Object Reference, ButtonText:Text=@n2) -> (then) {is=title K2Node_CallFunction}
   Event On Clicked()
   > n3 = Cast To FortPlayerControllerSpectating!(execute, Object=GetOwningPlayer()) -> (then, CastFailed, AsFort Player Controller Spectating:Fort Player Controller Spectating Object Reference) {is=title K2Node_DynamicCast}
   > FollowPlayerByName!(
    execute
    self:Fort Player Controller Spectating Object Reference=@n3
    PlayerName:String="To String (Text)"(InText=Get PlayerName() -> (PlayerName:Text))
   ) -> (then)
   Event Construct()
   > Parent: Construct()
  ]}]
variables [PlayerName {type=string, default=""}]
widgets {
 info {inheritedWidgetCount=0, parentClass=/FortUILibrary/Foundation/Buttons/IconTextButton.IconTextButton_C, rootWidgetClass=null, widgetCount=0}
 widgets []
}
