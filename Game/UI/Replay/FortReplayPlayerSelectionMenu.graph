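# Replay UI widget that lists the match's human players as a vertical menu of option entries inside a translucent dark border.
# On Construct, it casts the game state to FortGameState, loops over PlayerArray and filters out bots.
# Each remaining player gets a FortReplayPlayerOption added to PlayerList and labeled with the player name through SetupName.
# The list is built only on Construct, and this graph contains no selection handling.
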WidgetBlueprint /Game/UI/Replay/FortReplayPlayerSelectionMenu
graphs [EventGraph {nodes [
   Event Construct()
   > n1 = Cast To FortGameState!(execute, Object=GetGameState!() -> (ReturnValue:Game State Base Object Reference)) -> (then, CastFailed, AsFort Game State:Fort Game State Object Reference) {is=title K2Node_DynamicCast}
   > n2 = For Each Loop(Array:Array of Player State Object References=Get PlayerArray!(self:Game State Base Object Reference=@n1) -> (PlayerArray:Array of Player State Object References)) -> (Array Element:Player State Object Reference)
   > Branch(Condition=NOT Boolean(A=Get bIsABot!(self:Player State Object Reference=@n2) -> (bIsABot:bool)))
   n3 = "To Text (String)"(InString=GetPlayerName!(self:Player State Object Reference=@n2) -> (ReturnValue:String))
   > n4 = Create Widget!(
    execute:Exec
    Class:User Widget Class Reference=&/FortReplayPlayerOption.FortReplayPlayerOption_C
    OwningPlayer:Player Controller Object Reference
    ButtonText:Text=@n3
    IconBrush:Slate Brush Structure
    ControllerInputStyle:Common Button Style Class Reference=/FortUILibrary/Foundation/Buttons/ButtonStyles/ButtonStyle-TransparentWithCues-XS-B_OverlayYellow.ButtonStyle-TransparentWithCues-XS-B_OverlayYellow_C
    ButtonTextJustification:ETextJustify Enum=Center
    bDisplayAllCaps:bool=False
    Font:Slate Font Info Structure
    Style:Common Button Style Class Reference=&ButtonStyle-TransparentWithCues-XS-B_OverlayYellow_C
    bApplyAlphaOnDisable:bool=False
    bLocked:bool=False
    bSelectable:bool=False
    bShouldSelectUponReceivingFocus:bool=False
    bInteractableWhenSelected:bool=False
    bToggleable:bool=False
    bTriggerClickedAfterSelection:bool=False
    bDisplayInputActionWhenNotInteractable:bool=True
    bHideInputActionWithKeyboard:bool=False
    bShouldUseFallbackDefaultInputAction:bool=True
    bRequiresHold:bool=False
    HoldData:Common UIHold Data Class Reference
    ClickMethod:EButtonClickMethod Enum=DownAndUp
    InputPriority:int=0
    TriggeringInputAction:Data Table Row Handle Structure
   ) -> (then:Exec, ReturnValue:Fort Replay Player Option Object Reference)
   > AddChild(self=Get PlayerList() -> (PlayerList:Vertical Box Object Reference), Content=@n4)
   > SetupName!(execute, self:Fort Replay Player Option Object Reference=@n4, ButtonText:Text=@n3) -> (then)
  ]}]
variables []
widgets {
 info {inheritedWidgetCount=0, parentClass=&UserWidget, rootWidgetClass=&Border, widgetCount=2}
 widgets [
  Border_0 {widgetClassPath=&Border, widget=Border_0, properties {brushColor=<rgba 0.0176 0.0232 0.0331 0.5>, padding=<ltrb 0 0 0 0>}}
  PlayerList {
   widgetClassPath=&VerticalBox
   bIsVariable=true
   parent=Border_0
   widget=PlayerList
   slot {class=BorderSlot, properties {padding=<ltrb 0 0 0 0>}}
  }
 ]
}
