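# Replay timeline marker widget (FortReplayTimelineMarker subclass) that stacks a count label over an icon, defaulting to the kill icon.
# UpdateMarkerCount sets the label, which starts transparent, scales and fades in on hover and fades out on unhover.
# Unhover polls each frame until the hover animation's finished callback clears bAnimPlaying.
# Animation logic runs only when IsShippingBuild is false, and Construct also collapses the widget on that path.
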WidgetBlueprint /Game/UI/Replay/TimelineMarker
animations [
 OnHover {
  binding_count=2
  length_seconds=0.3
  object=&self:OnHover
  sequence {bindings [MarkerCount {
     id=#1
     possessed_class=CommonTextBlock
     tracks [
      {
       type=MovieScene2DTransformTrack
       sections [{
         type=MovieScene2DTransformSection
         end_frame=6
         start_frame=0
         channels [
          Translation.X {type=MovieSceneScriptingFloatChannel, default=0}
          Translation.Y {type=MovieSceneScriptingFloatChannel, default=0, const [0,2]}
          Angle {type=MovieSceneScriptingFloatChannel, default=0}
          Scale.X {type=MovieSceneScriptingFloatChannel, keys [[0,0.8],[9,1]]}
          Scale.Y {type=MovieSceneScriptingFloatChannel, keys [[0,0.8],[9,1]]}
          Shear.X {type=MovieSceneScriptingFloatChannel, default=0}
          Shear.Y {type=MovieSceneScriptingFloatChannel, default=0}
         ]
        }]
      }
      {
       type=MovieSceneFloatTrack
       display_name=Render Opacity
       sections [{type=MovieSceneFloatSection, end_frame=4, start_frame=0, channels [None {type=MovieSceneScriptingFloatChannel, keys [[0,0],[4,1]]}]}]
      }
     ]
    }]}
 }
 OnUnhover {
  binding_count=1
  length_seconds=0.15
  object=&self:OnUnhover
  sequence {bindings [MarkerCount {
     id=#2
     possessed_class=CommonTextBlock
     tracks [{
       type=MovieSceneFloatTrack
       display_name=Render Opacity
       sections [{type=MovieSceneFloatSection, end_frame=3, start_frame=0, channels [None {type=MovieSceneScriptingFloatChannel, keys [[0.75,1],[3,0]]}]}]
      }]
    }]}
 }
]
graphs [
 EventGraph {nodes [
   n1 = AnimDone() {is=title K2Node_CustomEvent}
   > Set bAnimPlaying(bAnimPlaying:bool=false) -> (Output_Get:bool)
   Event On Hovered()
   > n2 = Branch(Condition=IsShippingBuild!(self:Fort Runtime Options Object Reference=GetRuntimeOptions()) -> (ReturnValue:bool))
   Set bAnimPlaying(execute=@n2.else, bAnimPlaying=true)
   > PlayAnimationForward(InAnimation=Get OnHover() -> (OnHover:Widget Animation Object Reference), PlaybackSpeed=2)
   Event Construct()
   > n3 = Branch(Condition=IsShippingBuild(self=GetRuntimeOptions()))
   BindToAnimationFinished(execute=@n3.else, Animation=Get OnHover(), Delegate=@n1)
   > SetVisibility(InVisibility=Collapsed)
   n4 = Event On Unhovered()
   n5 = Branch(execute=[@n4.then, @n7.then], Condition=IsShippingBuild(self=GetRuntimeOptions()))
   n6 = Branch(execute=@n5.else, Condition=Get bAnimPlaying() -> (bAnimPlaying:bool))
   > n7 = DelayUntilNextFrame!(execute) -> (then)
   PlayAnimationForward(execute=@n6.else, InAnimation=Get OnUnhover() -> (OnUnhover:Widget Animation Object Reference), PlaybackSpeed=2)
  ]}
 UpdateMarkerCount {nodes [
   n1 = UpdateMarkerCount!() -> (then, Number:int) {is=title K2Node_FunctionEntry}
   > "SetText (Text)"(self=Get MarkerCount() -> (MarkerCount:Common Text Object Reference), InText="To Text (Integer)"(Value=@n1))
  ]}
]
variables [
 bAnimPlaying {type=boolean}
 HoveredColor {type=SlateColor, default {colorUseRule=UseColor_Specified, specifiedColor=<rgba 1 0.4969 0 1>}}
]
widgets {
 info {inheritedWidgetCount=2, parentClass=/Script/FortniteUI.FortReplayTimelineMarker, rootWidgetClass=&Overlay, widgetCount=6}
 widgets [
  Overlay_0 {widgetClassPath=&Overlay, widget=Overlay_0}
  VerticalBox_0 {
   widgetClassPath=&VerticalBox
   parent=Overlay_0
   widget=VerticalBox_0
   properties {bOverride_Cursor=true}
   slot {class=OverlaySlot, properties {horizontalAlignment=HAlign_Fill, verticalAlignment=VAlign_Fill, padding=<ltrb 2 2 2 2>}}
  }
  MarkerCount {
   widgetClassPath=&CommonTextBlock
   bIsVariable=true
   parent=VerticalBox_0
   widget=MarkerCount
   properties {
    justification=Center
    renderOpacity=0
    text=0
    font {
     bForceMonospaced=false
     bMaterialIsStencil=false
     fontMaterial=null
     fontObject=&HeadingNow
     letterSpacing=0
     monospacedWidth=1
     size=13.5
     skewAmount=0
     typefaceFontName=75 Medium
     outlineSettings {
      bApplyOutlineToDropShadows=true
      bMiteredCorners=false
      bSeparateFillAlpha=true
      outlineMaterial=&M-FontShadow
      outlineSize=3
      outlineColor=<rgba 0 0 0 0.3>
     }
    }
   }
   slot {class=VerticalBoxSlot, properties {horizontalAlignment=HAlign_Center, verticalAlignment=VAlign_Center, padding=<ltrb 0 0 0 3>}}
  }
  LazyImage_MarkerIcon {
   widgetClassPath=&FortLazyImage
   bInherited=true
   bIsVariable=true
   parent=VerticalBox_0
   widget=LazyImage_MarkerIcon
   properties {brush {
     bIsDynamicallyLoaded=false
     drawAs=Image
     imageType=NoImage
     mirroring=NoMirror
     resourceName=None
     resourceObject=/Game/UI/Foundation/Textures/Icons/Replay/T-Icon-Replay-Filter-Kill-M
     tiling=NoTile
     imageSize=<24 24>
     margin=<ltrb 0 0 0 0>
     outlineSettings {
      bUseBrushTransparency=false
      roundingType=HalfHeightRadius
      width=0
      color {colorUseRule=UseColor_Specified, specifiedColor=<rgba 0 0 0 0>}
      cornerRadii=<0 0 0 0>
     }
     tintColor {colorUseRule=UseColor_Specified, specifiedColor=<rgba 1 1 1 1>}
     uVRegion {bIsValid=false, max=<0 0>, min=<0 0>}
    }}
   slot {class=VerticalBoxSlot, properties {horizontalAlignment=HAlign_Center}}
  }
  Spacer_60 {widgetClassPath=&Spacer, parent=VerticalBox_0, widget=Spacer_60, properties {size=<0 5>}, slot {class=VerticalBoxSlot}}
  InputActionWidget {widgetClassPath=&CommonActionWidget, bInherited=true, widget=null}
 ]
}
